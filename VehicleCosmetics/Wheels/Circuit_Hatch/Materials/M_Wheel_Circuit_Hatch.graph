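# Unlit skeletal-mesh material for the Circuit Hatch vehicle wheel, drawing stylized hatched shading through emissive.
# Rim and tire masks and normals are blended, with the normals feeding a fresnel term.
# View-dependent and fresnel thresholds select layers of a panning, screen-aligned hatch texture, minus a fresnel edge and red masks, forming the VFX signal.
# Custom colors, a panning gradient and a Painted static switch build the rim and tread colors, scaled by Brightness.
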Material /VehicleCosmetics/Wheels/Circuit_Hatch/Materials/M_Wheel_Circuit_Hatch
graph {
 properties {
  bAutomaticallySetUsageInEditor=false
  bUseMaterialAttributes=true
  bUsedWithMeshDeformer=true
  bUsedWithSkeletalMesh=true
  previewMesh=/VehicleCosmetics/Wheels/Circuit_Hatch/Meshes/SK_Wheel_Circuit_Hatch
  shadingModel=MSM_Unlit
  usedShadingModels=Unlit
 }
 nodes [
  n1 = VectorParameter() {group=00 Base, parameterName=Custom Color, sortPriority=-1, defaultValue=<rgba 0 0.3278 0.117 1>}
  n2 = Clamp(None=ConstantBiasScale(None=@n1.RGB) {bias=0, scale=0.3})
  n3 = ScalarParameter() {defaultValue=0.05, group=00 Base, parameterName=Tread Flat Color}
  n4 = Clamp(None=Desaturation(None=ConstantBiasScale(None=@n1.RGB) {bias=0.3, scale=1}, Fraction=Constant() {r=-2}) {luminanceFactors=<rgba 0.3 0.59 0.11 0>})
  n5 = NamedRerouteUsage()
  n6 = TextureSample(UVs=Panner(Coordinate=Multiply(A=@n5) {constB=2}) {speedY=-0.5}) {samplerSource=SSM_Wrap_WorldGroupSettings, texture=/VehicleCosmetics/Wheels/_SharedTextures/FX/airmage_Gradiant_V}
  n7 = VectorParameter() {group=00 Base, parameterName=Custom Color 2, sortPriority=-1, defaultValue=<rgba 0.7835 0.7011 0.063 1>}
  n8 = LinearInterpolate(
   A=LinearInterpolate(
    A=StaticSwitchParameter(True=@n2, False=@n3) {group=00 Base, parameterName=Painted, sortPriority=0}
    B=LinearInterpolate(A=@n1.RGB, B=@n4, Alpha=@n6.R)
    Alpha=NamedRerouteUsage()
   )
   B=StaticSwitchParameter(True=@n4, False=@n7.RGB) {group=00 Base, parameterName=Painted, sortPriority=0}
   Alpha=NamedRerouteUsage()
  ) {desc="top = rim, bottom = tread"}
  n9 = NamedRerouteUsage()
  n10 = LinearInterpolate(A=Clamp(None=ConstantBiasScale(None=@n1.RGB) {bias=0, scale=0.075}), B=@n8, Alpha=Clamp(None=@n9)) {desc="top = rim, bottom = tread"}
  n11 = Multiply(A=@n8, B=Clamp(None=ConstantBiasScale(None=@n9) {bias=0.1, scale=1}))
  NamedRerouteDeclaration(None=Multiply(
    A=StaticSwitchParameter(True=@n10, False=@n11) {group=00 Base, parameterName=Painted, sortPriority=0}
    B=ScalarParameter() {defaultValue=1, group=00 Base, parameterName=Brightness}
   )) {name=Emissive, nodeColor=<rgba 1 0.4688 0 1>}
  NamedRerouteDeclaration(None=VertexColor().R) {name=VC_R, nodeColor=<rgba 0.6875 1 0 1>}
  n12 = TextureSampleParameter2D() {
   group=00 Base
   parameterName=Rim Masks
   samplerType=SAMPLERTYPE_LinearColor
   sortPriority=0
   texture=/VehicleCosmetics/Wheels/Circuit_Hatch/Textures/T_Circuit_Hatch_Rim_M
  }
  n13 = Add(A=TextureCoordinate(), B=AppendVector(A=Constant() {r=0.5}, B=Constant()))
  n14 = TextureSampleParameter2D(UVs=@n13) {
   group=00 Base
   parameterName=Tire Masks
   samplerType=SAMPLERTYPE_LinearColor
   sortPriority=1
   texture=/VehicleCosmetics/Wheels/Circuit_Hatch/Textures/Tire_Circuit_RGB
  }
  n15 = NamedRerouteUsage()
  NamedRerouteDeclaration(None=LinearInterpolate(A=@n12.G, B=@n14.G, Alpha=@n15)) {name=Masks G, nodeColor=<rgba 0 1 0 1>}
  NamedRerouteDeclaration(None=LinearInterpolate(A=@n12.B, B=@n14.B, Alpha=@n15)) {name=Masks B, nodeColor=<rgba 0 0 1 1>}
  n16 = MaterialFunctionCall!(Normal, Exponent=Constant() {r=3}) {materialFunction=/VehicleCosmetics/Wheels/Cola_Holographic/Materials/MF_Fresnel_Classic}
  n17 = Max(A=Power(Base=@n16.Result) {constExponent=1}, B=Clamp(None=ConstantBiasScale(None=@n16.Result) {bias=-0.35, scale=64}))
  n18 = CameraVectorWS()
  n19 = WorldPosition()
  n20 = DotProduct(
   A=Normalize(VectorInput=Transform(None=@n18) {transformSourceType=TRANSFORMSOURCE_World, transformType=TRANSFORM_Tangent})
   B=Normalize(VectorInput=Abs(None=@n19.XYZ))
  )
  n21 = TextureSampleParameter2D() {
   group=00 Base
   parameterName=Rim Normal
   samplerType=SAMPLERTYPE_Normal
   sortPriority=2
   texture=/VehicleCosmetics/Wheels/Circuit_Hatch/Textures/T_Circuit_Hatch_Rim_N
  }
  n22 = TextureSampleParameter2D(UVs=@n13) {
   group=00 Base
   parameterName=Tire Normal
   samplerType=SAMPLERTYPE_Normal
   sortPriority=3
   texture=/VehicleCosmetics/Wheels/Circuit_Hatch/Textures/Tire_Circuit_N
  }
  n23 = NamedRerouteUsage()
  n24 = ConstantBiasScale(None=MaterialFunctionCall!(
    Input Normal=LinearInterpolate(A=@n21.RGB, B=@n22.RGB, Alpha=@n23)
    Input Sharpness=ScalarParameter() {defaultValue=1, group=00 Base, parameterName=Fresnel Sharpness}
    Input Two Sided
   ) {materialFunction=/VehicleCosmetics/Wheels/CarCar_Holographic/Materials/MF_Fresnel_Advanced}.Edge) {bias=0.02, scale=1.2}
  n25 = ScalarParameter() {defaultValue=32, group=00 Base, parameterName=Fresnel Power}
  n26 = Subtract(A=Clamp(None=OneMinus(None=@n20)), B=Clamp(None=Power(Base=@n24, Exp=@n25) {constExponent=32}))
  n27 = ScreenPosition()
  n28 = Add(A=Multiply(A=ComponentMask(None=@n27.ViewportUV) {g=true, r=true}, B=Constant2Vector() {g=-1, r=1}), B=Constant2Vector() {g=1}) {desc=Achiving UE3 Screen Pos - Screen Aligned}
  n29 = Panner(Coordinate=Multiply(
    A=NamedRerouteDeclaration(None=@n28) {name=Screen Pos, nodeColor=<rgba 1 0 0.4922 1>}
    B=ScalarParameter() {defaultValue=1, group=00 Base, parameterName=UV Hatch Tiling}
   )) {speedX=0.02}
  n30 = TextureSample(UVs=@n29) {samplerSource=SSM_Wrap_WorldGroupSettings, texture=/VehicleCosmetics/Wheels/_SharedTextures/FX/skin_Jer_RGB}
  n31 = If(A=@n26, B=Constant() {r=0.75}, "A > B"=Constant() {r=1}, "A == B"=@n30.G, "A < B"=@n30.G) {constB=0.75}
  n32 = Min(
   A=@n31
   B=TextureSample(UVs=@n29) {samplerSource=SSM_Wrap_WorldGroupSettings, texture=&skin_Jer_RGB}.R
  )
  n33 = If(A=@n26, B=ScalarParameter() {defaultValue=0.3, group=00 Base, parameterName=Hatch Dist 1}, "A > B"=@n31, "A == B"=@n32, "A < B"=@n32) {constB=0.75}
  NamedRerouteDeclaration(None=Subtract(A=Subtract(A=@n33, B=@n17), B=LinearInterpolate(A=@n12.R, B=@n14.R, Alpha=@n15))) {name=VFX, nodeColor=<rgba 0.5469 0 1 1>}
  SetMaterialAttributes(Emissive Color=NamedRerouteUsage()) {attributeSetTypes [#1]}
 ]
}
